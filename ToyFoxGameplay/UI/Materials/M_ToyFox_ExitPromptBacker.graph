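# UI material that draws the backing panel for the ToyFox exit prompt, in dark-to-bright red tones with alpha-composite blending.
# An SDF box on 9-slice UVs yields an anti-aliased fill and a thin outline.
# A mirrored MSDF border texture supplies the bracket mask.
# Named reroutes carry the Box, Outline and Brackets masks into weighted sums for color and partial opacity.
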Material /ToyFoxGameplay/UI/Materials/M_ToyFox_ExitPromptBacker
graph {
 properties {bAutomaticallySetUsageInEditor=false, blendMode=BLEND_AlphaComposite, materialDomain=MD_UI, usedShadingModels=Default Lit}
 nodes [
  n1 = NamedRerouteUsage()
  n2 = Constant3Vector() {constant=<rgba 0.0391 0.0031 0.0088 1>}
  n3 = Multiply(A=NamedRerouteUsage())
  n4 = Constant3Vector() {constant=<rgba 0.467 0.0171 0.0639 1>}
  n5 = Multiply(A=NamedRerouteUsage()) {constB=0.3}
  Add(
   A=Multiply(A=Constant3Vector() {constant=<rgba 1 0.1022 0.1022 1>}, B=Multiply(B=@n1) {constA=1})
   B=Add(A=Multiply(A=@n2, B=@n3), B=Multiply(A=@n4, B=@n5))
  )
  n6 = MaterialFunctionCall!(UVs=MaterialFunctionCall() {materialFunction=&GetUserInterfaceUV}.9-Slice UV) {materialFunction=/FortUILibrary/MaterialFunctions/Basics/MF_UI_9SliceUVs}
  n7 = MaterialFunctionCall!(UVs=@n6.UVs, Size=Constant2Vector() {g=0.95, r=1}, CornerRoundness, KeepAspectRatio, PixelSize) {materialFunction=/FortUILibrary/MaterialFunctions/Shapes/SDFs/MF_SDF_Box}
  n8 = Multiply(
   A=MaterialFunctionCall(DistanceField=@n7.SDF) {materialFunction=&MF_UI_AntiAliasing}.Inner
   B=OneMinus(None=NamedRerouteUsage())
  )
  n9 = MaterialFunctionCall!(DistanceField=@n7.SDF, Threshold, Thickness=Constant() {r=0.07}, Alignment=Constant() {r=-1}, Smoothing) {materialFunction=/FortUILibrary/MaterialFunctions/Shapes/Modifiers/MF_UI_ShapeOutline}
  n10 = Multiply(A=@n9.Outline) {constB=0.1}
  Add(A=Multiply(A=NamedRerouteUsage()) {constB=0.5}, B=Multiply(A=Add(A=@n8, B=@n10)) {constB=0.5})
  Constant() {r=0.95}
  n11 = Constant2Vector() {g=1, r=1.5}
  n12 = TextureSample(UVs=MaterialFunctionCall!(
    UV=MaterialFunctionCall(UVs=@n6.UVs, Texture Scale=@n11) {materialFunction=&ScaleUVsByCenter}.UVs
    Mirror Y
    Swizzle UV
   ) {materialFunction=/FortUILibrary/MaterialFunctions/Basics/MF_UI_Symmetry_2x1_New}.Symmetry) {samplerType=SAMPLERTYPE_LinearColor, texture=/ToyFoxGameplay/UI/Materials/Textures/T_UI_ToyFox_ExitBorder}
  NamedRerouteDeclaration(None=MaterialFunctionCall(
    DistanceField=MaterialFunctionCall!(MSDF RGB=@n12.RGB) {materialFunction=/FortUILibrary/MaterialFunctions/Shapes/Modifiers/MF_UI_MSDFtoSDF}.SDF
    Threshold=Constant() {r=0.4}
   ) {materialFunction=&MF_UI_AntiAliasing}.Outer) {name=Brackets, nodeColor=<rgba 0.2188 0 1 1>}
  NamedRerouteDeclaration(None=@n9.Outline) {name=Outline, nodeColor=<rgba 0.1016 1 0 1>}
  NamedRerouteDeclaration(None=@n8) {name=Box, nodeColor=<rgba 1 0 0.4219 1>}
 ]
}
