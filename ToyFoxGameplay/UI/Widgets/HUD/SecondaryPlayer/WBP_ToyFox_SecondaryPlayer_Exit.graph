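# HUD element widget that shows an "Exit (Hold)" prompt for the ToyFox secondary player.
# An overlay layers a material backer behind a row with a red-tinted exit icon, label and scaled-down keybind widget.
# On Construct, after the parent call, it overrides the keybind's input context with IMC_ToyFox_Secondary and binds IA_ToyFox_Secondary_Exit, so the displayed key follows that mapping.
# The graph only configures the display and contains no exit or hold logic.
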WidgetBlueprint /ToyFoxGameplay/UI/Widgets/HUD/SecondaryPlayer/WBP_ToyFox_SecondaryPlayer_Exit
graphs [EventGraph {nodes [
   Event Construct()
   > Parent: Construct()
   n1 = Get Keybind() -> (Keybind:Keybind Widget Object Reference)
   > SetInputContextOverride!(
    execute
    self:Fort Keybind Widget Object Reference=@n1
    NewContextOverride:Input Mapping Context Object Reference=/ToyFoxGameplay/Input/IMC_ToyFox_Secondary
   ) -> (then)
   > SetBoundInputAction!(
    execute
    self:Fort Keybind Widget Object Reference=@n1
    NewAction:Input Action Object Reference=/ToyFoxGameplay/Input/IA_ToyFox_Secondary_Exit
   ) -> (then)
  ]}]
variables []
widgets {
 info {inheritedWidgetCount=1, parentClass=&FortHUDElementWidget, rootWidgetClass=&Overlay, widgetCount=9}
 widgets [
  Overlay_50 {widgetClassPath=&Overlay, widget=Overlay_50}
  Image_Backer {
   widgetClassPath=&Image
   bIsVariable=true
   parent=Overlay_50
   widget=Image_Backer
   properties {brush {
     bIsDynamicallyLoaded=false
     drawAs=Image
     imageType=FullColor
     mirroring=NoMirror
     resourceName=None
     resourceObject=/ToyFoxGameplay/UI/Materials/M_ToyFox_ExitPromptBacker
     tiling=NoTile
     imageSize=<32 32>
     margin=<ltrb 0 0 0 0>
     outlineSettings {
      bUseBrushTransparency=false
      roundingType=HalfHeightRadius
      width=0
      color {colorUseRule=UseColor_Specified, specifiedColor=<rgba 0 0 0 0>}
      cornerRadii=<0 0 0 0>
     }
     tintColor {colorUseRule=UseColor_Specified, specifiedColor=<rgba 1 1 1 1>}
     uVRegion {bIsValid=false, max=<0 0>, min=<0 0>}
    }}
   slot {class=OverlaySlot, properties {horizontalAlignment=HAlign_Fill, verticalAlignment=VAlign_Fill, padding=<ltrb -5 -2 -5 -2>}}
  }
  HorizontalBox_0 {
   widgetClassPath=&HorizontalBox
   parent=Overlay_50
   widget=HorizontalBox_0
   properties {bOverride_Cursor=true}
   slot {class=OverlaySlot}
  }
  Image_Icon {
   widgetClassPath=&Image
   bIsVariable=true
   parent=HorizontalBox_0
   widget=Image_Icon
   properties {brush {
     bIsDynamicallyLoaded=false
     drawAs=Image
     imageType=FullColor
     mirroring=NoMirror
     resourceName=None
     resourceObject=/ToyFoxGameplay/UI/Materials/Textures/T_UI_ToyFox_Exit
     tiling=NoTile
     imageSize=<35 35>
     margin=<ltrb 0 0 0 0>
     outlineSettings {
      bUseBrushTransparency=false
      roundingType=HalfHeightRadius
      width=0
      color {colorUseRule=UseColor_Specified, specifiedColor=<rgba 0 0 0 0>}
      cornerRadii=<0 0 0 0>
     }
     tintColor {colorUseRule=UseColor_Specified, specifiedColor=<rgba 1 0.3776 0.4025 1>}
     uVRegion {bIsValid=false, max=<0 0>, min=<0 0>}
    }}
   slot {class=HorizontalBoxSlot}
  }
  CommonTextBlock_342 {
   widgetClassPath=&CommonTextBlock
   parent=HorizontalBox_0
   widget=CommonTextBlock_342
   properties {
    justification=Center
    text="Exit (Hold)"
    colorAndOpacity {colorUseRule=UseColor_Specified, specifiedColor=<rgba 1 0.7991 0.7991 1>}
    font {
     bForceMonospaced=false
     bMaterialIsStencil=false
     fontMaterial=null
     fontObject=&BurbankSmall
     letterSpacing=0
     monospacedWidth=1
     size=12
     skewAmount=0
     typefaceFontName=Black
     outlineSettings {
      bApplyOutlineToDropShadows=false
      bMiteredCorners=false
      bSeparateFillAlpha=false
      outlineMaterial=null
      outlineSize=0
      outlineColor=<rgba 0 0 0 1>
     }
    }
    shadowOffset=<0 0>
   }
   slot {class=HorizontalBoxSlot, properties {verticalAlignment=VAlign_Center, padding=<ltrb 8 0 0 0>}}
  }
  Spacer_41 {widgetClassPath=&Spacer, parent=HorizontalBox_0, widget=Spacer_41, properties {size=<10 1>}, slot {class=HorizontalBoxSlot}}
  SizeBox_0 {
   widgetClassPath=&SizeBox
   parent=HorizontalBox_0
   widget=SizeBox_0
   properties {bOverride_HeightOverride=true, bOverride_WidthOverride=true, heightOverride=56, widthOverride=56}
   slot {class=HorizontalBoxSlot, properties {horizontalAlignment=HAlign_Right, verticalAlignment=VAlign_Center, padding=<ltrb -12 -12 -12 -12>}}
  }
  Keybind {
   widgetClassPath=/FortUILibrary/Input/KeybindWidget
   bIsVariable=true
   parent=SizeBox_0
   widget=Keybind
   properties {bUseKeyboardTextureForKeys=true, brushSize=Small, renderTransform {angle=0, scale=<0.8 0.8>, shear=<0 0>, translation=<0 0>}}
   slot {class=SizeBoxSlot, properties {horizontalAlignment=HAlign_Center, verticalAlignment=VAlign_Center}}
  }
  TutorialHighlight {widgetClassPath=&TutorialHighlight}
 ]
}
