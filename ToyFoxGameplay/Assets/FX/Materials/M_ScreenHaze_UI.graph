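# Unlit alpha-composite UI material for ToyFox that draws an animated screen-edge haze with a pulsing red warning state.
# MPC_ScreenHazeUp supplies target-axis offset, camera direction and warning amount, which shift the radial vignette and light side edges.
# The vignette masks a breathing color gradient, scrolling hex pattern, side glows and radial wave lines.
# Platform switches zero out noise distortion and wave lines on last-gen consoles, Switch, Android and PC performance mode.
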Material /ToyFoxGameplay/Assets/FX/Materials/M_ScreenHaze_UI
graph {
 properties {
  bAutomaticallySetUsageInEditor=false
  bDisableDepthTest=true
  bUsedWithNiagaraMeshParticles=true
  bUsedWithNiagaraRibbons=true
  bUsedWithNiagaraSprites=true
  blendMode=BLEND_AlphaComposite
  blendableLocation=BL_SceneColorBeforeBloom
  materialDomain=MD_UI
  shadingModel=MSM_Unlit
  usedShadingModels=Unlit
 }
 nodes [
  n1 = MaterialFunctionCall(Float2=TextureCoordinate()) {materialFunction=&BreakOutFloat2Components}
  Add(A=@n1.G) {constB=0.5}
  Constant()
  n2 = VectorParameter() {parameterName=ColorC, defaultValue=<rgba 1 0.05 0.6745 1>}
  n3 = CollectionParameter() {collection=&/MPC_ScreenHazeUp, parameterName=LengthBetweenTargetAxis}
  n4 = Multiply(
   A=Add(A=@n3, B=ScalarParameter() {defaultValue=-1, group=BoardWarningTh, parameterName=BoardWarningTh_GrayAdjust_Add})
   B=ScalarParameter() {defaultValue=5, group=BoardWarningTh, parameterName=BoardWarningTh_GrayAdjust_Mul}
  )
  n5 = Saturate(None=Power(Base=Saturate(None=@n4), Exp=ScalarParameter() {defaultValue=3, group=BoardWarningTh, parameterName=BoardWarningTh_GrayAdjust_Power}))
  n6 = Constant()
  n7 = TextureObjectParameter() {
   group=Noise
   parameterName=Noise_Texture
   samplerType=SAMPLERTYPE_Normal
   sortPriority=0
   texture=/Game/Characters/Player/Female/Medium/Bodies/F_MED_Bistro_Astronaut/FX/T_GaussianNoise_Normal
  }
  n8 = VectorParameter() {group=Noise, parameterName=Noise_UVControl, channelNames {a=VPan, b=UPan, g=VScale, r=UScale}, defaultValue=<rgba 1 0.7 0 0.2>}
  n9 = VectorParameter() {
   group=Noise
   parameterName=Noise_NoiseControl
   sortPriority=60
   channelNames {a=VStrength, b=UStrength, g=VOffset, r=UOffset}
   defaultValue=<rgba 0 0 0.07 0.07>
  }
  n10 = MaterialFunctionCall!(
   NoiseValue=MaterialFunctionCall(Texture=@n7, UVControl_RGB=@n8, UVControl_A=@n8, "[Override]UVInput"=TextureCoordinate()) {materialFunction=&FXCommon_TextureSample}.RG
   Control_RGB=@n9
   Control_A=@n9
  ) {materialFunction=/Game/Effects/Niagara/MasterMaterials/Functions/Common/FXCommon_NoiseControl}
  n11 = TextureCoordinate()
  n12 = VectorParameter() {parameterName=MainMS, defaultValue=<rgba 1 1.5 0 1>}
  n13 = ComponentMask(None=AppendVector(A=@n12.R, B=@n12.R)) {g=true, r=true}
  n14 = MaterialFunctionCall(Float3=CollectionParameter() {collection=&MPC_ScreenHazeUp, parameterName=CameraDirection}) {materialFunction=&BreakOutFloat3Components}
  n15 = MaterialFunctionCall(X=@n14.G, Y=@n14.G, Z=@n14.G) {materialFunction=&MakeFloat3}
  n16 = ComponentMask(None=@n15.Result) {g=true, r=true}
  n17 = ScalarParameter() {defaultValue=-0.2, parameterName=MoveStrength}
  n18 = ScalarParameter() {defaultValue=0.2, parameterName=MaxMove}
  n19 = Multiply(A=Normalize(VectorInput=@n16), B=Clamp(None=Multiply(A=@n3, B=@n17), Min=Multiply(A=@n18) {constB=-1}, Max=@n18)) {constB=0.5}
  n20 = Add(
   A=Add(
    A=MaterialFunctionCall!("Xb1,PS4"=@n6, "PC-High&&Epic,XSX,PS5"=@n10.NoiseValue, "PC-Perf,Android"=@n6, Switch=@n6, "PC-Low(NotPerf)&&Medium"=@n10.NoiseValue) {materialFunction=/Game/Effects/Niagara/MasterMaterials/Functions/Common/PlatformSwitch_2D}.Result
    B=MaterialFunctionCall(UVs=@n11, Texture Scale=@n13) {materialFunction=&ScaleUVsByCenter}.UVs
   )
   B=Multiply(A=ComponentMask(None=@n19) {g=true, r=true}, B=Constant2Vector() {r=1})
  )
  n21 = OneMinus(None=SphereMask(A=@n20, B=Constant2Vector() {g=0.5, r=0.5}, Radius=ScalarParameter() {defaultValue=0.7, parameterName=CRad}) {attenuationRadius=1, hardnessPercent=0})
  n22 = ScalarParameter() {defaultValue=3, parameterName=PowerToTheColor}
  n23 = Time()
  n24 = ScalarParameter() {defaultValue=0.05, parameterName=ColorGradBreathTimer}
  n25 = Abs(None=Sine(None=Multiply(A=@n23, B=@n24)))
  n26 = ScalarParameter() {defaultValue=0.85, parameterName=ColorGradBreathOffset}
  n27 = Divide(A=Saturate(None=Subtract(A=@n25, B=@n26)), B=OneMinus(None=@n26))
  n28 = Add(
   A=Add(A=Power(Base=@n21, Exp=@n22) {constExponent=3}, B=ScalarParameter() {defaultValue=-0.1, parameterName=ColorOOOO})
   B=Multiply(A=Power(Base=@n27) {constExponent=1.5}, B=ScalarParameter() {defaultValue=0.15, parameterName=ScaleGradBrathOffset})
  )
  n29 = Constant3Vector() {constant=<rgba 1 0 0 0>}
  n30 = Constant3Vector() {constant=<rgba 0.2 0 0 0>}
  n31 = CollectionParameter() {collection=&MPC_ScreenHazeUp, parameterName=WarningStart}
  n32 = Time()
  n33 = ScalarParameter() {defaultValue=1, parameterName=EgeBTimer}
  n34 = Abs(None=Sine(None=Multiply(A=@n32, B=@n33)))
  n35 = LinearInterpolate(
   A=MaterialFunctionCall(
    A=VectorParameter() {parameterName=ColorA, defaultValue=<rgba 0 0.1183 1 1>}.RGB
    B=VectorParameter() {parameterName=ColorB, defaultValue=<rgba 0.1395 0.044 1.1 1>}.RGB
    C=LinearInterpolate(A=@n2.RGB, B=@n2.RGB, Alpha=@n5)
    Alpha=Saturate(None=@n28)
   ) {materialFunction=&3ColorBlend}.Result
   B=LinearInterpolate(A=@n29, B=LinearInterpolate(A=@n29, B=@n30, Alpha=@n31), Alpha=@n34)
   Alpha=@n31
  )
  MaterialFunctionCall(Texture=@n35) {materialFunction=&HueShift}
  MaterialFunctionCall!(Vector3=@n15.Result, MaximumNumberOfDigits, UVs, DebugTextLocation RG_UpperRight BA_LowerLeft, Component Spacing) {materialFunction=/Engine/Functions/Engine_MaterialFunctions02/Utility/DebugFloat3Values}
  n36 = Constant2Vector() {g=0.5, r=0.5}
  n37 = ScalarParameter() {defaultValue=0.75, parameterName=HexUVRad}
  n38 = Power(
   Base=OneMinus(None=SphereMask(A=@n11, B=@n36, Radius=@n37) {attenuationRadius=1, hardnessPercent=0})
   Exp=ScalarParameter() {defaultValue=2.5, parameterName=HexUVPower}
  ) {constExponent=3}
  n39 = TextureCoordinate()
  n40 = Constant2Vector() {g=0.5, r=0.5}
  n41 = Abs(None=Multiply(A=@n3, B=ScalarParameter() {defaultValue=0.7, parameterName=MoveToHexSTE}))
  n42 = Sine(None=Multiply(A=Time(), B=ScalarParameter() {defaultValue=0.25, parameterName=HexBreathTime}))
  n43 = Constant() {r=-1}
  n44 = Constant() {r=1}
  n45 = VectorParameter() {parameterName=HexBreath, defaultValue=<rgba 0.8 1 0.7 1>}
  n46 = MaterialFunctionCall(Input=@n42, Input Low=@n43, Input High=@n44, Target Low=@n45.R, Target High=@n45.R) {materialFunction=&RemapValueRange}
  n47 = Multiply(
   A=Normalize(VectorInput=Subtract(A=@n39, B=@n40))
   B=LinearInterpolate(
    A=ScalarParameter() {defaultValue=-0.2, parameterName=STE}
    B=ScalarParameter() {defaultValue=-1, parameterName=STE2}
    Alpha=Multiply(A=@n41, B=@n46.Result)
   )
  )
  n48 = MaterialFunctionCall(Float2=ViewProperty() {property=MEVP_ViewSize}.Property) {materialFunction=&BreakOutFloat2Components}
  n49 = Divide(A=@n48.R, B=@n48.R)
  n50 = VectorParameter() {group=HexTexture, parameterName=HexTexture_UVControl, channelNames {a=VPan, b=UPan, g=VScale, r=UScale}, defaultValue=<rgba 2 2.5 0 0>}
  n51 = MaterialFunctionCall(
   UVControl_RGB=MaterialFunctionCall(X=Multiply(A=@n49, B=@n50.UScale), Y=@n50.VScale, Z=@n50.VScale) {materialFunction=&MakeFloat3}.Result
   UVControl_A=@n50.VPan
  ) {materialFunction=&FXCommon_TextureSample}
  n52 = VectorParameter() {parameterName=SpeedHex, defaultValue=<rgba 0 0.02 0 1>}
  n53 = Multiply(
   A=MaterialFunctionCall(X=@n52.R, Y=@n52.R) {materialFunction=&MakeFloat2}.Result
   B=Time() {bOverride_Period=true, period=6000}
  )
  n54 = TextureObjectParameter() {
   group=HexTexture
   parameterName=HexTexture_Texture
   samplerType=SAMPLERTYPE_Alpha
   sortPriority=0
   texture=/Game/Effects/Fort_Effects/GameplayPulginsAssets/ToyFox/T_UI_ToyFox_Pattern_02
  }
  n55 = TextureSample(UVs=Add(A=Multiply(A=@n38, B=@n47), B=Add(A=@n51.UV, B=@n53)), Tex=@n54)
  n56 = Multiply(
   A=Add(A=Desaturation(None=@n55.RGB), B=ScalarParameter() {group=HexTexture, parameterName=HexTexture_GrayAdjust_Add})
   B=ScalarParameter() {defaultValue=1, group=HexTexture, parameterName=HexTexture_GrayAdjust_Mul}
  )
  n57 = Saturate(None=Power(Base=Saturate(None=@n56), Exp=ScalarParameter() {defaultValue=1, group=HexTexture, parameterName=HexTexture_GrayAdjust_Power}))
  n58 = VectorParameter() {parameterName=HexRef_Refine, defaultValue=<rgba 5 10 2 0.5>}
  n59 = OneMinus(None=Power(Base=OneMinus(None=@n57), Exp=@n58.R))
  n60 = LinearInterpolate(A=Saturate(None=Multiply(A=@n59, B=@n58.G)), B=Power(Base=Multiply(A=@n57, B=@n58.B)) {constExponent=0.5}, Alpha=@n58.A)
  n61 = CollectionParameter() {collection=&MPC_ScreenHazeUp, parameterName=WarningStart}
  n62 = MaterialFunctionCall(
   Input=OneMinus(None=Saturate(None=@n60))
   Input Low=Constant()
   Input High=Constant() {r=1}
   Target Low=Multiply(A=LinearInterpolate(Alpha=@n61) {constA=1, constB=0.65}, B=ScalarParameter() {defaultValue=0.45, parameterName=RWR})
   Target High=MaterialFunctionCall(Input=@n42, Input Low=@n43, Input High=@n44, Target Low=@n45.B, Target High=@n45.B) {materialFunction=&RemapValueRange}.Result
  ) {materialFunction=&RemapValueRange}
  n63 = Constant2Vector() {g=0.5, r=0.5}
  n64 = ScalarParameter() {defaultValue=0.6, parameterName=HexRad}
  n65 = Power(
   Base=OneMinus(None=SphereMask(A=@n20, B=@n63, Radius=@n64) {attenuationRadius=1, hardnessPercent=0})
   Exp=ScalarParameter() {defaultValue=4, parameterName=HexPower}
  ) {constExponent=3}
  n66 = Multiply(A=@n1.G) {constB=0.5}
  n67 = VectorParameter() {parameterName=curve, defaultValue=<rgba -24.0281 1.0046 0 1>}
  n68 = Multiply(
   A=MaterialFunctionCall(Input=Sine(None=@n66), Input Low=@n67.R, Input High=@n67.R) {materialFunction=&RemapValue_0-1}.Result
   B=ScalarParameter() {defaultValue=-0.8064, group=EdgeBright, parameterName=EdgeBright_GrayAdjust_Add}
  )
  n69 = ScalarParameter() {defaultValue=5, group=EdgeBright, parameterName=EdgeBright_GrayAdjust_Mul}
  n70 = Multiply(A=Add(A=OneMinus(None=@n1.R), B=@n68), B=@n69)
  n71 = ScalarParameter() {defaultValue=4, group=EdgeBright, parameterName=EdgeBright_GrayAdjust_Power}
  n72 = Saturate(None=Power(Base=Saturate(None=@n70), Exp=@n71))
  n73 = VectorParameter() {parameterName=EdgeBright_Refine, defaultValue=<rgba 9.5591 1 0.1 0.5>}
  n74 = ScalarParameter() {defaultValue=0.5, parameterName=EdgeBrightOffsetC}
  n75 = Subtract(A=Add(A=@n3, B=ScalarParameter() {defaultValue=-0.5, parameterName=EdgeBrightStartOffset}), B=@n74)
  n76 = Multiply(
   A=Sign(None=MaterialFunctionCall(Float2=@n16) {materialFunction=&BreakOutFloat2Components}.R)
   B=Divide(A=Saturate(None=@n75), B=OneMinus(None=@n74))
  )
  n77 = ScalarParameter() {defaultValue=1, parameterName=Sca}
  n78 = Multiply(A=Clamp(None=@n76), B=@n77)
  n79 = Multiply(A=Add(A=@n1.R, B=@n68), B=@n69)
  n80 = Saturate(None=Power(Base=Saturate(None=@n79), Exp=@n71))
  n81 = Multiply(A=Clamp(None=Multiply(A=@n76) {constB=-1}), B=@n77)
  n82 = Add(
   A=Multiply(
    A=MaterialFunctionCall(Texture=@n72, RGB=@n73.RGB, A=@n73.RGB) {materialFunction=&MF_FX_Refine}.Result
    B=Saturate(None=@n78)
   )
   B=Multiply(
    A=MaterialFunctionCall(Texture=@n80, RGB=@n73.RGB, A=@n73.RGB) {materialFunction=&MF_FX_Refine}.Result
    B=Saturate(None=@n81)
   )
  )
  n83 = Saturate(None=Multiply(A=@n82, B=ScalarParameter() {defaultValue=20, parameterName=HexMsk2}))
  n84 = ScalarParameter() {defaultValue=1, parameterName=HexMask2Pow}
  n85 = Constant()
  n86 = TextureCoordinate()
  n87 = MaterialFunctionCall(Float2=@n86) {materialFunction=&BreakOutFloat2Components}
  n88 = VectorParameter() {group=UMask, parameterName=UMask_EdgeMask_InvertControl, channelNames {a="", b=Power, g=Multiply, r=Offset}, defaultValue=<rgba 0 2.5 1.5 1>}
  n89 = OneMinus(None=SphereMask(A=@n86, B=Constant2Vector() {g=0.5, r=0.5}, Radius=ScalarParameter() {defaultValue=0.8, parameterName=Rad5}) {attenuationRadius=1, hardnessPercent=0})
  n90 = ScalarParameter() {defaultValue=2.9458, parameterName=Pow5}
  n91 = Multiply(
   A=Add(A=TextureCoordinate()) {constB=-0.5}
   B=ComponentMask(None=VectorParameter() {parameterName=et4, defaultValue=<rgba 2 1 0 1>}.RGB) {g=true, r=true}
  )
  n92 = Add(
   A=MaterialFunctionCall(Float2=MaterialFunctionCall(Vector or UVs=@n91) {materialFunction=&VectorToRadialValue}.Radial Coordinates) {materialFunction=&BreakOutFloat2Components}.G
   B=Multiply(A=Time(), B=ScalarParameter() {defaultValue=-0.01, parameterName=WaveLineOffset})
  )
  n93 = ScalarParameter() {defaultValue=30, parameterName=WaveLineScale}
  n94 = Abs(None=Sine(None=Multiply(A=@n92, B=@n93)))
  n95 = ScalarParameter() {group=WaveLine, parameterName=WaveLine_GrayAdjust_Add}
  n96 = Saturate(None=Multiply(A=Add(A=@n94, B=@n95), B=ScalarParameter() {defaultValue=1, group=WaveLine, parameterName=WaveLine_GrayAdjust_Mul}))
  n97 = ScalarParameter() {defaultValue=300.3369, group=WaveLine, parameterName=WaveLine_GrayAdjust_Power}
  n98 = Multiply(A=@n35, B=Saturate(None=Power(Base=@n96, Exp=@n97)))
  n99 = VectorParameter() {parameterName=WaveLine, defaultValue=<rgba 1.2183 1.2357 1.5 1>}
  n100 = Multiply(
   A=OneMinus(None=MaterialFunctionCall!(UorV=@n87.R, Invert_EdgeControl=@n88, NoInvert_EdgeControl=@n88) {materialFunction=/Game/Effects/Niagara/MasterMaterials/Functions/Common/FXCommon_EdgeMask_Ex}.Mask)
   B=Multiply(A=Power(Base=@n89, Exp=@n90) {constExponent=3}, B=Multiply(A=@n98, B=@n99.RGB))
  )
  n101 = CollectionParameter() {collection=&MPC_ScreenHazeUp, parameterName=WarningStart}
  n102 = Multiply(
   A=LinearInterpolate(
    A=VectorParameter() {parameterName=EdgeBrightColor, defaultValue=<rgba 15 2.5 8.75 1>}.RGB
    B=VectorParameter() {parameterName=EdgeBrightColor_Warning, defaultValue=<rgba 10 1.5 1.5001 1>}.RGB
    Alpha=@n101
   )
   B=LinearInterpolate(A=Constant() {r=1}, B=Constant(), Alpha=@n34)
  )
  n103 = Add(
   A=MaterialFunctionCall!("Xb1,PS4"=@n85, "PC-High&&Epic,XSX,PS5"=@n100, "PC-Perf,Android"=@n85, Switch=@n85, "PC-Low(NotPerf)&&Medium"=@n100) {materialFunction=/Game/Effects/Niagara/MasterMaterials/Functions/Common/PlatformSwitch_3D}.Result
   B=Add(A=Multiply(A=@n82, B=@n102), B=@n35)
  )
  n104 = Constant2Vector() {g=0.5, r=0.5}
  n105 = ScalarParameter() {defaultValue=0.6, parameterName=Rad}
  n106 = Power(
   Base=OneMinus(None=SphereMask(A=@n20, B=@n104, Radius=@n105) {attenuationRadius=1, hardnessPercent=0})
   Exp=ScalarParameter() {defaultValue=8, parameterName=Pow}
  ) {constExponent=3}
  Multiply(
   A=LinearInterpolate(A=LinearInterpolate(B=@n62.Result, Alpha=@n65) {constA=1}, Alpha=Power(Base=@n83, Exp=@n84))
   B=Multiply(A=Multiply(A=@n103, B=@n106), B=ScalarParameter() {defaultValue=1, parameterName=PostStrength})
  )
  Multiply(A=@n11, B=@n13)
  OneMinus(None=Power(Base=OneMinus(None=@n3), Exp=ScalarParameter() {defaultValue=2, parameterName=sPoweLenForHex}))
  n107 = ScalarParameter() {defaultValue=0.5, parameterName=OpacityScale}
  Reroute(None=Multiply(A=@n106, B=LinearInterpolate(A=@n107, Alpha=@n101) {constB=0.25}))
 ]
}
