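# HUD frame for the secondary player of the giant Megazord pawn, layering lens flare, screen haze and frame material images.
# Construct caches the giant pawn from the ToyFox game state component and the owning controller.
# Tick feeds the camera-versus-pawn direction into MPC_ScreenHazeUp and shifts the frame by their clamped rotation delta, using a curve-table yaw limit for both axes.
# A damage animation drives frame glitch and flare parameters but is never played here; the critical animation is empty.
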WidgetBlueprint /ToyFoxGameplay/UI/Widgets/HUD/SecondaryPlayer/WBP_ToyFox_SecondaryPlayer_HUDFrame
animations [
 Anim_DamageSustained {
  binding_count=2
  length_seconds=0.5
  object=&self:Anim_DamageSustained
  sequence {
   bindings [
    Image_Frame {
     id=#1
     possessed_class=Image
     tracks [{
       type=MovieSceneWidgetMaterialTrack
       sections [{
         type=MovieSceneParameterSection
         end_frame=30
         start_frame=0
         channels [Glitch In {type=MovieSceneScriptingFloatChannel, keys [[0,1],[30,0]]}]
        }]
      }]
    }
    Image_LensFlare {
     id=#2
     possessed_class=Image
     tracks [{
       type=MovieSceneWidgetMaterialTrack
       sections [{
         type=MovieSceneParameterSection
         end_frame=30
         start_frame=0
         channels [Starbust Location {type=MovieSceneScriptingFloatChannel, keys [[0,0.324],[30,0.8851]]}]
        }]
      }]
    }
   ]
   display_rate {numerator=60}
  }
 }
 Anim_Critical {
  binding_count=0
  length_seconds=0.55
  object=&self:Anim_Critical
  sequence {bindings [], display_rate {numerator=60}}
 }
]
graphs [EventGraph {nodes [
   Event Construct()
   n1 = GetGameStateAthena!() -> (ReturnValue:Fort Game State Athena Object Reference)
   > Set MegazordPawn(MegazordPawn:Fort Giant Player Pawn Object Reference=GetGiantPawn!(self:B Toy Fox Game State Component Object Reference=GetComponentByClass!(
      self:Actor Object Reference=@n1
      ComponentClass:Actor Component Class Reference=/ToyFoxGameplay/Gameplay/B_ToyFoxGameStateComponent.B_ToyFoxGameStateComponent_C
     ) -> (ReturnValue:B Toy Fox Game State Component Object Reference)) -> (Pawn:Toy Fox Giant Player Pawn Object Reference)) -> (Output_Get:Fort Giant Player Pawn Object Reference)
   > Set SecondaryPawn(SecondaryPawn:Player Controller Object Reference=GetOwningPlayer()) -> (Output_Get:Player Controller Object Reference)
   Event Tick()
   n2 = Get PlayerCameraManager!(self:Player Controller Object Reference=GetFortPlayerControllerFromActor!(Actor:Actor Object Reference=Get SecondaryPawn() -> (SecondaryPawn:Player Controller Object Reference)) -> (ReturnValue:Fort Player Controller Object Reference)) -> (PlayerCameraManager:Player Camera Manager Object Reference)
   n3 = Invert Rotator!(A:Rotator=GetCameraRotation!(self:Player Camera Manager Object Reference=@n2) -> (ReturnValue:Rotator)) -> (ReturnValue:Rotator)
   n4 = Get Actor Rotation!(self:Actor Object Reference=Get MegazordPawn() -> (MegazordPawn:Fort Giant Player Pawn Object Reference)) -> (ReturnValue:Rotator) {is=title K2Node_CallFunction}
   n5 = GetCameraRotation(self=@n2)
   n6 = vector - vector(
    A=GetAxes!(A:Rotator=@n4) -> (X:Vector, Y:Vector, Z:Vector)
    B=Get Rotation X Vector!(InRot:Rotator=@n5) -> (ReturnValue:Vector) {is=title K2Node_CallFunction}
   )
   > SetVectorParameterValue!(
    execute
    Collection:Material Parameter Collection Object Reference=/ToyFoxGameplay/Assets/FX/Materials/MPC_ScreenHazeUp
    ParameterName:Name=CameraDirection
    ParameterValue:"Linear Color Structure (by ref)"="To LinearColor (Vector)"!(InVec:Vector=Rotate Vector!(A:Vector=@n6, B:Rotator=@n3) -> (ReturnValue:Vector)) -> (ReturnValue:Linear Color Structure)
   ) -> (then) {is=title K2Node_CallMaterialParameterCollectionFunction}
   > SetScalarParameterValue!(
    execute
    Collection:Material Parameter Collection Object Reference=&MPC_ScreenHazeUp
    ParameterName:Name=LengthBetweenTargetAxis
    ParameterValue:float=Vector Length(A=@n6)
   ) -> (then) {is=title K2Node_CallMaterialParameterCollectionFunction}
   n7 = BreakRotator(InRot="Delta (Rotator)"!(A:Rotator=@n4, B:Rotator=@n5) -> (ReturnValue:Rotator))
   n8 = MakeVector2D!(X=@n7.Yaw, Y=@n7.Pitch) -> (ReturnValue_X:double, ReturnValue_Y:double)
   n9 = GetCachedScalableFloatValue!(Input:"Fort Cached Scalable Float Structure (by ref)"=Get Hotfix_CameraRotMaxYaw() -> (Hotfix_CameraRotMaxYaw:Fort Cached Scalable Float Structure)) -> (ReturnValue:float)
   n10 = float * float(A=@n9, B=-1) {is=title K2Node_PromotableOperator}
   > SetRenderTranslation!(
    execute
    self=Get Image_Frame() -> (Image_Frame:Image Object Reference)
    Translation_X:double=MapRangeClamped(Value=@n8, InRangeA=@n10, InRangeB=@n9, OutRangeA=-150, OutRangeB=150)
    Translation_Y:double=MapRangeClamped(Value=@n8.ReturnValue_Y, InRangeA=@n10, InRangeB=@n9, OutRangeA=-20, OutRangeB=20)
   ) -> (then)
   Event PreConstruct()
  ]}]
variables [
 Hotfix_CameraRotMaxYaw {
  type=FortCachedScalableFloat
  default {scalableFloat {
    registryType=None
    value=1
    curve {curveTable=/ToyFoxGameplay/DataTables/ToyFoxGameplayGameData, rowName=ToyFox.SecondaryPlayer.Camera.Rot.Yaw.Max}
   }}
 }
 MegazordPawn {type=FortGiantPlayerPawn, default=null}
 SecondaryPawn {type=PlayerController, default=null}
]
widgets {
 info {inheritedWidgetCount=1, parentClass=&FortHUDElementWidget, rootWidgetClass=&Overlay, widgetCount=5}
 widgets [
  Overlay_28 {widgetClassPath=&Overlay, widget=Overlay_28}
  Image_LensFlare {
   widgetClassPath=&Image
   bIsVariable=true
   parent=Overlay_28
   widget=Image_LensFlare
   properties {
    renderOpacity=0
    visibility=SelfHitTestInvisible
    brush {
     bIsDynamicallyLoaded=false
     drawAs=Image
     imageType=FullColor
     mirroring=NoMirror
     resourceName=None
     resourceObject=/ToyFoxGameplay/UI/Widgets/HUD/Assets/HUDFrame/M_UI_ToyFox_ScreenGlare
     tiling=NoTile
     imageSize=<32 32>
     margin=<ltrb 0 0 0 0>
     outlineSettings {
      bUseBrushTransparency=false
      roundingType=HalfHeightRadius
      width=0
      color {colorUseRule=UseColor_Specified, specifiedColor=<rgba 0 0 0 0>}
      cornerRadii=<0 0 0 0>
     }
     tintColor {colorUseRule=UseColor_Specified, specifiedColor=<rgba 1 1 1 1>}
     uVRegion {bIsValid=false, max=<0 0>, min=<0 0>}
    }
    renderTransform {angle=0, scale=<1.4 1.2>, shear=<0 0>, translation=<0 0>}
   }
   slot {class=OverlaySlot, properties {horizontalAlignment=HAlign_Fill, verticalAlignment=VAlign_Fill}}
  }
  Image_ScreenHaze {
   widgetClassPath=&Image
   bIsVariable=true
   parent=Overlay_28
   widget=Image_ScreenHaze
   properties {
    visibility=SelfHitTestInvisible
    brush {
     bIsDynamicallyLoaded=false
     drawAs=Image
     imageType=FullColor
     mirroring=NoMirror
     resourceName=None
     resourceObject=/ToyFoxGameplay/Assets/FX/Materials/M_ScreenHaze_UI
     tiling=NoTile
     imageSize=<1920 1080>
     margin=<ltrb 0 0 0 0>
     outlineSettings {
      bUseBrushTransparency=false
      roundingType=HalfHeightRadius
      width=0
      color {colorUseRule=UseColor_Specified, specifiedColor=<rgba 0 0 0 0>}
      cornerRadii=<0 0 0 0>
     }
     tintColor {colorUseRule=UseColor_Specified, specifiedColor=<rgba 1 1 1 1>}
     uVRegion {bIsValid=false, max=<0 0>, min=<0 0>}
    }
   }
   slot {class=OverlaySlot, properties {horizontalAlignment=HAlign_Fill, verticalAlignment=VAlign_Fill}}
  }
  Image_Frame {
   widgetClassPath=&Image
   bIsVariable=true
   parent=Overlay_28
   widget=Image_Frame
   properties {
    visibility=SelfHitTestInvisible
    brush {
     bIsDynamicallyLoaded=false
     drawAs=Image
     imageType=FullColor
     mirroring=NoMirror
     resourceName=None
     resourceObject=/ToyFoxGameplay/UI/Widgets/HUD/Assets/HUDFrame/M_UI_HUDFrame2
     tiling=NoTile
     imageSize=<1920 1080>
     margin=<ltrb 0 0 0 0>
     outlineSettings {
      bUseBrushTransparency=false
      roundingType=HalfHeightRadius
      width=0
      color {colorUseRule=UseColor_Specified, specifiedColor=<rgba 0 0 0 0>}
      cornerRadii=<0 0 0 0>
     }
     tintColor {colorUseRule=UseColor_Specified, specifiedColor=<rgba 1 1 1 1>}
     uVRegion {bIsValid=false, max=<0 0>, min=<0 0>}
    }
    renderTransform {angle=0, scale=<1.3 1.1>, shear=<0 0>, translation=<0 0>}
   }
   slot {class=OverlaySlot, properties {horizontalAlignment=HAlign_Fill, verticalAlignment=VAlign_Fill}}
  }
  TutorialHighlight {widgetClassPath=&TutorialHighlight}
 ]
}
