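# HUD quickbar widget for the ToyFox secondary player, showing Basic Rocket and Seeking Rocket slots plus rocket ammo.
# Construct retries every 0.2 seconds until the owning pawn's secondary pilot component is valid, then binds ammo changes and fills the count and maximum text.
# Tick converts the reload timer into radial slider progress, though that slider is collapsed and transparent by default, and prints the value on screen as debug output.
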WidgetBlueprint /ToyFoxGameplay/UI/Widgets/HUD/SecondaryPlayer/WBP_ToyFox_SecondaryPlayer_Quickbar
graphs [EventGraph {nodes [
   n1 = UpdateAmmo!() -> (OutputDelegate, then, AmmoCount:int) {is=title K2Node_CustomEvent}
   > Set BasicRocketLatestReplicatedAmmoCount(BasicRocketLatestReplicatedAmmoCount:int=@n1.AmmoCount) -> (Output_Get:int=0)
   > "SetText (Text)"(self=Get AmmoCount() -> (AmmoCount:Common Text Object Reference), InText="To Text (Integer)"(Value=@n1.AmmoCount))
   Event Construct()
   > n2 = Parent: Construct()
   n3 = GetOwningPlayerPawn!(self:User Widget Object Reference) -> (ReturnValue:Pawn Object Reference)
   n4 = Is Valid(exec=[@n2.then, @n7.then], InputObject=@n3)
   > n5 = GetComponentByClass!(
    execute
    self:Actor Object Reference=@n3
    ComponentClass:Actor Component Class Reference=/ToyFoxGameplay/Gameplay/PawnComp/B_ToyFoxPilotPawnComponent_Secondary.B_ToyFoxPilotPawnComponent_Secondary_C
   ) -> (then, ReturnValue:B Toy Fox Pilot Pawn Component Secondary Object Reference)
   > n6 = Is Valid(InputObject=@n5)
   n7 = Delay(execute=[@n6.Is Not Valid, @n4.Is Not Valid], Duration=0.2)
   Set SecondaryPilotPawnComponent(execute=@n6.Is Valid, SecondaryPilotPawnComponent:B Toy Fox Pilot Pawn Component Secondary Object Reference=@n5) -> (Output_Get:B Toy Fox Pilot Pawn Component Secondary Object Reference)
   > Bind Event to On Basic Rocket Ammo Count Changed(self:B Toy Fox Pilot Pawn Component Secondary Object Reference=@n5, Delegate=Create Event()) {is=title K2Node_AddDelegate}
   > UpdateAmmo!(
    execute
    self
    AmmoCount:int=GetCurrentBasicRocketAmmoCount!(self:B Toy Fox Pilot Pawn Component Secondary Object Reference=@n5) -> (Count:int=0)
   ) -> (then) {is=title K2Node_CallFunction}
   n8 = GetMaxBasicRocketAmmoCount!(self:B Toy Fox Pilot Pawn Component Secondary Object Reference=@n5) -> (ReturnValue:int=0)
   > Set BasicRocketMaxAmmoCount(BasicRocketMaxAmmoCount:int=@n8) -> (Output_Get:int=0)
   > "SetText (Text)"(self=Get MaxAmmo() -> (MaxAmmo:Common Text Object Reference), InText=Format Text!(Format="LOCTEXT(\"/ {max}\")", max:int=@n8) -> (Result))
   n9 = Event Tick()
   > Parent: Tick!(execute, MyGeometry:Geometry Structure=@n9.MyGeometry, InDeltaTime:float=@n9.InDeltaTime) -> (then) {is=title K2Node_CallParentFunction}
   > n10 = Get!(execute:Exec) -> (then:Exec, else:Exec, SecondaryPilotPawnComponent:B Toy Fox Pilot Pawn Component Secondary Object Reference)
   > n11 = GetBasicRocketAmmoReloadTimerHandle!(execute, self:B Toy Fox Pilot Pawn Component Secondary Object Reference=@n10) -> (then, TimerHandle:Timer Handle Structure)
   n12 = Get Timer Elapsed Time by Handle!(Handle:Timer Handle Structure=@n11) -> (ReturnValue:float) {is=title K2Node_CallFunction}
   n13 = SafeDivide(
    A=@n12
    B=float + float(A=@n12, B=Get Timer Remaining Time by Handle!(Handle:Timer Handle Structure=@n11) -> (ReturnValue:float) {is=title K2Node_CallFunction})
   )
   > SetValue!(
    execute
    self:Radial Slider Object Reference=Get RadialSlider_BasicRocket() -> (RadialSlider_BasicRocket:Radial Slider Object Reference)
    InValue:float=@n13
   ) -> (then)
   > PrintString!(
    execute
    InString=Append(A="UpdateRadialSlideBasicRocket: ", B="To String (Float)"(InDouble=@n13))
    bPrintToScreen
    bPrintToLog=false
    TextColor
    Duration
    Key=UpdateRadialSlideBasicRocket
   ) -> (then)
  ]}]
variables [
 BasicRocketLatestReplicatedAmmoCount {type=integer}
 BasicRocketMaxAmmoCount {type=integer}
 BasicRocketUpdateRadialSliderTimerHandle {type=TimerHandle}
 SecondaryPilotPawnComponent {type=B_ToyFoxPilotPawnComponent_Secondary_C, default=null}
]
widgets {
 info {
  inheritedWidgetCount=1
  parentClass=&FortHUDElementWidget
  rootWidgetClass=/Script/FortniteUI.CommonTagVisibilityWidget
  widgetCount=16
 }
 widgets [
  CommonTagVisibilityWidget {
   widgetClassPath=&CommonTagVisibilityWidget
   bIsVariable=true
   widget=CommonTagVisibilityWidget
   properties {visibility=HitTestInvisible}
  }
  SizeBox_0 {
   widgetClassPath=&SizeBox
   parent=CommonTagVisibilityWidget
   widget=SizeBox_0
   properties {bOverride_WidthOverride=true, widthOverride=580}
   slot {class=PanelSlot}
  }
  Overlay_16 {widgetClassPath=&Overlay, parent=SizeBox_0, widget=Overlay_16, slot {class=SizeBoxSlot}}
  Image_Border {
   widgetClassPath=&Image
   bIsVariable=true
   parent=Overlay_16
   widget=Image_Border
   properties {
    brush {
     bIsDynamicallyLoaded=false
     drawAs=Image
     imageType=FullColor
     mirroring=NoMirror
     resourceName=None
     resourceObject=/ToyFoxGameplay/UI/Materials/M_ToyFox_HealthTopBar
     tiling=NoTile
     imageSize=<128 128>
     margin=<ltrb 0 0 0 0>
     outlineSettings {
      bUseBrushTransparency=false
      roundingType=HalfHeightRadius
      width=0
      color {colorUseRule=UseColor_Specified, specifiedColor=<rgba 0 0 0 0>}
      cornerRadii=<0 0 0 0>
     }
     tintColor {colorUseRule=UseColor_Specified, specifiedColor=<rgba 1 1 1 1>}
     uVRegion {bIsValid=false, max=<0 0>, min=<0 0>}
    }
    renderTransform {angle=0, scale=<-1 1>, shear=<0 0>, translation=<26 90>}
   }
   slot {class=OverlaySlot, properties {horizontalAlignment=HAlign_Fill, verticalAlignment=VAlign_Fill}}
  }
  GridPanel {
   widgetClassPath=&GridPanel
   parent=Overlay_16
   widget=GridPanel
   properties {renderTransform {angle=0, scale=<0.8 0.8>, shear=<0 0>, translation=<0 0>}}
   slot {class=OverlaySlot, properties {horizontalAlignment=HAlign_Right, verticalAlignment=VAlign_Center}}
  }
  Overlay_2 {
   widgetClassPath=&Overlay
   parent=GridPanel
   widget=Overlay_2
   properties {renderTransform {angle=4, scale=<1 1>, shear=<0 0>, translation=<-69 0>}}
   slot {class=GridSlot, properties {horizontalAlignment=HAlign_Left, verticalAlignment=VAlign_Bottom, padding=<ltrb 0 0 0 10>}}
  }
  Image_Bg {
   widgetClassPath=&Image
   bIsVariable=true
   parent=Overlay_2
   widget=Image_Bg
   properties {
    renderOpacity=0.8
    brush {
     bIsDynamicallyLoaded=false
     drawAs=Image
     imageType=FullColor
     mirroring=NoMirror
     resourceName=None
     resourceObject=/ToyFoxGameplay/UI/Materials/MI_ToyFox_AbilityBg
     tiling=NoTile
     imageSize=<90 35>
     margin=<ltrb 0 0 0 0>
     outlineSettings {
      bUseBrushTransparency=false
      roundingType=HalfHeightRadius
      width=0
      color {colorUseRule=UseColor_Specified, specifiedColor=<rgba 0 0 0 0>}
      cornerRadii=<0 0 0 0>
     }
     tintColor {colorUseRule=UseColor_Specified, specifiedColor=<rgba 1 1 1 1>}
     uVRegion {bIsValid=false, max=<0 0>, min=<0 0>}
    }
    renderTransform {angle=-177, scale=<1.9 1.1>, shear=<-39 0>, translation=<-20 -2>}
   }
   slot {class=OverlaySlot, properties {horizontalAlignment=HAlign_Fill, verticalAlignment=VAlign_Center}}
  }
  HorizontalBox_0 {
   widgetClassPath=&HorizontalBox
   parent=Overlay_2
   widget=HorizontalBox_0
   properties {renderTransform {angle=28, scale=<1 1>, shear=<0 -25>, translation=<0 -3>}}
   slot {class=OverlaySlot, properties {horizontalAlignment=HAlign_Center, verticalAlignment=VAlign_Fill}}
  }
  AmmoCount {
   widgetClassPath=&CommonTextBlock
   bIsVariable=true
   parent=HorizontalBox_0
   widget=AmmoCount
   properties {
    text=12
    colorAndOpacity {colorUseRule=UseColor_Specified, specifiedColor=<rgba 1 0.7991 0.7991 1>}
    font {
     bForceMonospaced=false
     bMaterialIsStencil=false
     fontMaterial=null
     fontObject=&Burbank
     letterSpacing=0
     monospacedWidth=1
     size=24
     skewAmount=0
     typefaceFontName=Burbank-Black
     outlineSettings {
      bApplyOutlineToDropShadows=false
      bMiteredCorners=false
      bSeparateFillAlpha=false
      outlineMaterial=&M-FontShadow
      outlineSize=3
      outlineColor=<rgba 0 0 0 0.5>
     }
    }
    shadowOffset=<0 0>
   }
   slot {class=HorizontalBoxSlot, properties {verticalAlignment=VAlign_Center}}
  }
  MaxAmmo {
   widgetClassPath=&CommonTextBlock
   bIsVariable=true
   parent=HorizontalBox_0
   widget=MaxAmmo
   properties {
    text=/ 12
    colorAndOpacity {colorUseRule=UseColor_Specified, specifiedColor=<rgba 1 0.7991 0.7991 1>}
    font {
     bForceMonospaced=false
     bMaterialIsStencil=false
     fontMaterial=null
     fontObject=&Burbank
     letterSpacing=0
     monospacedWidth=1
     size=24
     skewAmount=0
     typefaceFontName=Burbank-Black
     outlineSettings {
      bApplyOutlineToDropShadows=false
      bMiteredCorners=false
      bSeparateFillAlpha=false
      outlineMaterial=&M-FontShadow
      outlineSize=3
      outlineColor=<rgba 0 0 0 0.5>
     }
    }
    shadowOffset=<0 0>
   }
   slot {class=HorizontalBoxSlot, properties {verticalAlignment=VAlign_Center, padding=<ltrb 4 0 0 0>}}
  }
  Overlay_0 {
   widgetClassPath=&Overlay
   parent=GridPanel
   widget=Overlay_0
   properties {renderTransform {angle=0, scale=<1 1>, shear=<0 0>, translation=<33 9>}}
   slot {class=GridSlot, properties {padding=<ltrb 0 0 -20 0>}}
  }
  RadialSlider_BasicRocket {
   widgetClassPath=/Script/AdvancedWidgets.RadialSlider
   bIsVariable=true
   parent=Overlay_0
   widget=RadialSlider_BasicRocket
   properties {
    renderOpacity=0
    sliderHandleEndAngle=360
    sliderHandleStartAngle=0
    stepSize=0.0596
    visibility=Collapsed
    sliderProgressColor=<rgba 0 1 0.6848 1>
    widgetStyle {
     barThickness=9.511
     disabledBarImage {
      bIsDynamicallyLoaded=false
      drawAs=RoundedBox
      imageType=NoImage
      mirroring=NoMirror
      resourceName=None
      resourceObject=null
      tiling=NoTile
      imageSize=<0 0>
      margin=<ltrb 0 0 0 0>
      outlineSettings {
       bUseBrushTransparency=false
       roundingType=FixedRadius
       width=1
       color {colorUseRule=UseColor_ColorTable, specifiedColor=<rgba 0 0 0 0>}
       cornerRadii=<2 2 2 2>
      }
      tintColor {colorUseRule=UseColor_Specified, specifiedColor=<rgba 0.0048 0.0048 0.0048 1>}
      uVRegion {bIsValid=false, max=<0 0>, min=<0 0>}
     }
     disabledThumbImage {
      bIsDynamicallyLoaded=false
      drawAs=RoundedBox
      imageType=NoImage
      mirroring=NoMirror
      resourceName=None
      resourceObject=null
      tiling=NoTile
      imageSize=<8 8>
      margin=<ltrb 0 0 0 0>
      outlineSettings {
       bUseBrushTransparency=false
       roundingType=HalfHeightRadius
       width=0
       color {colorUseRule=UseColor_Specified, specifiedColor=<rgba 0 0 0 0>}
       cornerRadii=<0 0 0 0>
      }
      tintColor {colorUseRule=UseColor_Specified, specifiedColor=<rgba 0.5271 0.5271 0.5271 1>}
      uVRegion {bIsValid=false, max=<0 0>, min=<0 0>}
     }
     hoveredBarImage {
      bIsDynamicallyLoaded=false
      drawAs=RoundedBox
      imageType=NoImage
      mirroring=NoMirror
      resourceName=None
      resourceObject=null
      tiling=NoTile
      imageSize=<0 0>
      margin=<ltrb 0 0 0 0>
      outlineSettings {
       bUseBrushTransparency=false
       roundingType=FixedRadius
       width=1
       color {colorUseRule=UseColor_ColorTable, specifiedColor=<rgba 0 0 0 0>}
       cornerRadii=<2 2 2 2>
      }
      tintColor {colorUseRule=UseColor_Specified, specifiedColor=<rgba 0.0048 0.0048 0.0048 1>}
      uVRegion {bIsValid=false, max=<0 0>, min=<0 0>}
     }
     hoveredThumbImage {
      bIsDynamicallyLoaded=false
      drawAs=RoundedBox
      imageType=NoImage
      mirroring=NoMirror
      resourceName=None
      resourceObject=null
      tiling=NoTile
      imageSize=<8 8>
      margin=<ltrb 0 0 0 0>
      outlineSettings {
       bUseBrushTransparency=false
       roundingType=HalfHeightRadius
       width=0
       color {colorUseRule=UseColor_Specified, specifiedColor=<rgba 0 0 0 0>}
       cornerRadii=<0 0 0 0>
      }
      tintColor {colorUseRule=UseColor_Specified, specifiedColor=<rgba 1 1 1 1>}
      uVRegion {bIsValid=false, max=<0 0>, min=<0 0>}
     }
     normalBarImage {
      bIsDynamicallyLoaded=false
      drawAs=RoundedBox
      imageType=NoImage
      mirroring=NoMirror
      resourceName=None
      resourceObject=null
      tiling=NoTile
      imageSize=<0 0>
      margin=<ltrb 0 0 0 0>
      outlineSettings {
       bUseBrushTransparency=false
       roundingType=FixedRadius
       width=1
       color {colorUseRule=UseColor_ColorTable, specifiedColor=<rgba 0 0 0 0>}
       cornerRadii=<2 2 2 2>
      }
      tintColor {colorUseRule=UseColor_Specified, specifiedColor=<rgba 0.0048 0.0048 0.0048 1>}
      uVRegion {bIsValid=false, max=<0 0>, min=<0 0>}
     }
     normalThumbImage {
      bIsDynamicallyLoaded=false
      drawAs=RoundedBox
      imageType=NoImage
      mirroring=NoMirror
      resourceName=None
      resourceObject=null
      tiling=NoTile
      imageSize=<8 8>
      margin=<ltrb 0 0 0 0>
      outlineSettings {
       bUseBrushTransparency=false
       roundingType=HalfHeightRadius
       width=0
       color {colorUseRule=UseColor_Specified, specifiedColor=<rgba 0 0 0 0>}
       cornerRadii=<0 0 0 0>
      }
      tintColor {colorUseRule=UseColor_Specified, specifiedColor=<rgba 0.5271 0.5271 0.5271 1>}
      uVRegion {bIsValid=false, max=<0 0>, min=<0 0>}
     }
    }
   }
   slot {class=OverlaySlot}
  }
  QuickbarSlot_BasicRocket {
   widgetClassPath=/ToyFoxGameplay/UI/Widgets/HUD/WBP_ToyFox_QuickbarSlot
   bIsVariable=true
   parent=Overlay_0
   widget=QuickbarSlot_BasicRocket
   properties {
    abilityIcon=/ToyFoxGameplay/UI/Materials/MI_ToyFox_ZordAbilityIcons_BasicRocket
    inputBoundAction=Fire
    cooldownGameplayTag {tagName=ToyFox.Ability.Cooldown.Rocket.Explosive}
    renderTransform {angle=0, scale=<1 1>, shear=<0 0>, translation=<0 10>}
   }
   slot {class=OverlaySlot, properties {horizontalAlignment=HAlign_Center, verticalAlignment=VAlign_Center}}
  }
  Overlay_1 {
   widgetClassPath=&Overlay
   parent=GridPanel
   widget=Overlay_1
   properties {renderTransform {angle=0, scale=<1 1>, shear=<0 0>, translation=<0 29>}}
   slot {class=GridSlot, properties {column=1, padding=<ltrb 25 0 0 0>}}
  }
  QuickbarSlot_SeekingRocket {
   widgetClassPath=&WBP_ToyFox_QuickbarSlot
   bIsVariable=true
   parent=Overlay_1
   widget=QuickbarSlot_SeekingRocket
   properties {
    abilityIcon=/ToyFoxGameplay/UI/Materials/MI_ToyFox_ZordAbilityIcons_SeekingRocket
    bShowCooldownSlider=true
    inputBoundAction=Target
    cooldownGameplayTag {tagName=ToyFox.Ability.Cooldown.Rocket.SeekingMissile}
    renderTransform {angle=0, scale=<1 1>, shear=<0 0>, translation=<0 8>}
   }
   slot {class=OverlaySlot, properties {horizontalAlignment=HAlign_Center, verticalAlignment=VAlign_Center}}
  }
  TutorialHighlight {widgetClassPath=&TutorialHighlight}
 ]
}
